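# Sound cue for the close sound of the weapon mod station (bench), routed to the non-critical in-game UI sound class.
# A single WeaponModBench_Close wave player feeds a Modulator node left at its default settings.
# A concurrency override caps it at two simultaneous instances, stopping the oldest, with a short retrigger time.
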SoundCue /WeaponModStation/Sounds/WeaponModStation_Close_Cue
root=Modulator_0
nodes [
 {id=Modulator_0, children [WavePlayer_1]}
 {id=WavePlayer_1, properties {soundWaveAssetPtr=/WeaponModStation/Sounds/Wavs/WeaponModBench_Close}}
]
properties {
 bOverrideConcurrency=true
 maxDistance=2097152
 soundClassObject=&Fort_UI_InGame_NonCritical
 volumeMultiplier=1
 concurrencyOverrides {maxCount=2, resolutionRule=StopOldest, retriggerTime=0.25}
}
