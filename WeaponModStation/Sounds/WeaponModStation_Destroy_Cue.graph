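# Sound cue for the destruction of the Weapon Mod Station, blending vending machine metal impacts with a piñata explosion.
# A Mixer layers one of three equally weighted random impact waves with the explosion, each branch through a Modulator.
# An Enveloper ramps the explosion's volume up from silence.
# Overridden attenuation adds logarithmic falloff and listener focus, while overridden concurrency allows one instance per owner, stopping the quietest.
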SoundCue /WeaponModStation/Sounds/WeaponModStation_Destroy_Cue
root=Mixer_0
nodes [
 {
  id=Enveloper_0
  children [Modulator_4]
  properties {pitchCurve {editorCurveData {keys [{time=0}]}}, volumeCurve {editorCurveData {keys [<at 0 0>,<at 0.7 1>]}}}
 }
 {id=Mixer_0, children [Modulator_5,Enveloper_0], properties {inputVolume [1,1]}}
 {id=Modulator_4, children [WavePlayer_14]}
 {id=Modulator_5, children [Random_4]}
 {id=Random_4, children [WavePlayer_15, WavePlayer_16, WavePlayer_17], properties {weights [1,1,1]}}
 {id=WavePlayer_14, properties {soundWaveAssetPtr=/Game/Sounds/Fort_Human_VO/Pinata/Pinata_New_Explode}}
 {
  id=WavePlayer_15
  properties {soundWaveAssetPtr=/Game/Sounds/Fort_Impact_Sounds/Demolition/VendingMachine/Metal_VendMachine_ImpactFinal_mix_01}
 }
 {
  id=WavePlayer_16
  properties {soundWaveAssetPtr=/Game/Sounds/Fort_Impact_Sounds/Demolition/VendingMachine/Metal_VendMachine_ImpactFinal_mix_02}
 }
 {
  id=WavePlayer_17
  properties {soundWaveAssetPtr=/Game/Sounds/Fort_Impact_Sounds/Demolition/VendingMachine/Metal_VendMachine_ImpactFinal_mix_03}
 }
]
properties {
 attenuationSettings=/Game/Sounds/Attenuation/Fort_Item_Interact_Att
 bOverrideAttenuation=true
 bOverrideConcurrency=true
 maxDistance=1000
 soundClassObject=&Fort_Loot_Reward_PSM
 volumeMultiplier=0.8
 attenuationOverrides {
  bEnableFocusInterpolation=true
  bEnableListenerFocus=true
  distanceAlgorithm=Logarithmic
  falloffDistance=1000
  focusAttackInterpSpeed=20
  focusAzimuth=20
  focusReleaseInterpSpeed=10
  focusVolumeAttenuation=1.6
  lPFFrequencyAtMax=10000
  lPFRadiusMax=1000
  lPFRadiusMin=500
  nonFocusAzimuth=40
  nonFocusVolumeAttenuation=0.6
  nonSpatializedRadiusStart=200
  occlusionInterpolationTime=0.3
  occlusionLowPassFilterFrequency=3000
  occlusionVolumeAttenuation=0.75
  attenuationShapeExtents {x=0}
  customAttenuationCurve {editorCurveData {keys [<at 0 1 RCIM_Cubic>, {arriveTangent=-1, interpMode=RCIM_Cubic, leaveTangent=-1, time=0.5, value=0.1}, <at 1 0 RCIM_Cubic>]}}
 }
 concurrencyOverrides {bLimitToOwner=true, maxCount=1, resolutionRule=StopQuietest}
}
