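# Sound cue that plays when a magazine mod is equipped at the Weapon Mod Station.
# It plays a single wave, the mod bench's magazine purchase sound, straight from a wave player with no randomization or modulation.
# It uses the item interact attenuation and routes through the loot reward sound class.
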SoundCue /WeaponModStation/Sounds/WeaponModStation_Equip_Magazines_Cue
root=WavePlayer_4
nodes [{id=WavePlayer_4, properties {soundWaveAssetPtr=/WeaponModStation/Sounds/Wavs/WeaponModBench_PurchaseMod_Magazine}}]
properties {
 attenuationSettings=/Game/Sounds/Attenuation/Fort_Item_Interact_Att
 maxDistance=1850
 soundClassObject=&Fort_Loot_Reward_PSM
 volumeMultiplier=1
}
