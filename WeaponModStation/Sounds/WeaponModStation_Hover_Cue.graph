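# Sound cue that plays the hover feedback for the weapon mod station UI.
# A mixer layers two front-end button hover waves at different levels, fed through a modulator whose pitch and volume ranges are fixed, so it adds no variation.
# It routes to the non-critical in-game UI sound class, with concurrency rules that keep rapid hovers from stacking by stopping the oldest.
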SoundCue /WeaponModStation/Sounds/WeaponModStation_Hover_Cue
root=Modulator_1
nodes [
 {id=Mixer_0, children [WavePlayer_0,WavePlayer_1], properties {inputVolume [0.75,0.5]}}
 {id=Modulator_1, children [Mixer_0], properties {pitchMax=1, pitchMin=1, volumeMax=1, volumeMin=1}}
 {id=WavePlayer_0, properties {soundWaveAssetPtr=/Game/Sounds/Fort_UI_Sounds/FrontEnd/New_FrontEnd/UI_FN_Button_Hover_Click_02}}
 {id=WavePlayer_1, properties {soundWaveAssetPtr=/Game/Sounds/Fort_UI_Sounds/FrontEnd/New_FrontEnd/UI_FN_Button_Hover}}
]
properties {
 maxDistance=2097152
 soundClassObject=&Fort_UI_InGame_NonCritical
 volumeMultiplier=1.4
 concurrencyOverrides {maxCount=1, resolutionRule=StopOldest, retriggerTime=0.15}
 concurrencySet [/Game/Sounds/Concurrency_Rules/FrontEnd_UI_Menu_Max_01_Stop_Oldest_Hover_Tick, &/Paprika_WeaponModStation_Select_Conc]
}
