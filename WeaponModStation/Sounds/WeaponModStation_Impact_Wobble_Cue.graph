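# Sound Cue for the Weapon Mod Station's impact wobble, built from borrowed Wumba container foley.
# A Mixer layers one randomly chosen, equally weighted rattle from three variants over a light-flash sound with modulated pitch.
# Attenuation overrides add short-range logarithmic falloff, listener focus, distance low-pass and occlusion settings.
# A concurrency override caps simultaneous instances per owner, stopping the oldest, and playback routes through a loot-reward sound class.
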SoundCue /WeaponModStation/Sounds/WeaponModStation_Impact_Wobble_Cue
root=Mixer_0
nodes [
 {id=Mixer_0, children [Modulator_0,Random_0], properties {inputVolume [1,1]}}
 {id=Modulator_0, children [WavePlayer_4], properties {pitchMin=0.9}}
 {id=Random_0, children [WavePlayer_1,WavePlayer_2,WavePlayer_3], properties {weights [1,1,1]}}
 {id=WavePlayer_1, properties {soundWaveAssetPtr=/Game/Sounds/Foley_Loot/Containers/Wumba/Wumba_Impact_Rattle}}
 {id=WavePlayer_2, properties {soundWaveAssetPtr=/Game/Sounds/Foley_Loot/Containers/Wumba/Wumba_Impact_Rattle_02}}
 {id=WavePlayer_3, properties {soundWaveAssetPtr=/Game/Sounds/Foley_Loot/Containers/Wumba/Wumba_Impact_Rattle_03}}
 {id=WavePlayer_4, properties {soundWaveAssetPtr=/Game/Sounds/Foley_Loot/Containers/Wumba/Wumba_Light_Flash}}
]
properties {
 attenuationSettings=/Game/Sounds/Attenuation/Fort_Item_Interact_Att
 bOverrideAttenuation=true
 bOverrideConcurrency=true
 maxDistance=1000
 soundClassObject=&Fort_Loot_Reward_PSM
 volumeMultiplier=0.8
 attenuationOverrides {
  bEnableFocusInterpolation=true
  bEnableListenerFocus=true
  distanceAlgorithm=Logarithmic
  falloffDistance=1000
  focusAttackInterpSpeed=20
  focusAzimuth=20
  focusReleaseInterpSpeed=10
  focusVolumeAttenuation=1.6
  lPFFrequencyAtMax=10000
  lPFRadiusMax=1000
  lPFRadiusMin=500
  nonFocusAzimuth=40
  nonFocusVolumeAttenuation=0.6
  nonSpatializedRadiusStart=200
  occlusionInterpolationTime=0.3
  occlusionLowPassFilterFrequency=3000
  occlusionVolumeAttenuation=0.75
  attenuationShapeExtents {x=0}
  customAttenuationCurve {editorCurveData {keys [<at 0 1 RCIM_Cubic>, {arriveTangent=-1, interpMode=RCIM_Cubic, leaveTangent=-1, time=0.5, value=0.1}, <at 1 0 RCIM_Cubic>]}}
 }
 concurrencyOverrides {bLimitToOwner=true, maxCount=3, resolutionRule=StopOldest}
}
