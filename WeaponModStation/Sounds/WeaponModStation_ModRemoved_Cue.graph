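# Sound cue for removing a mod at the weapon mod station, going by its name, assigned to the non-critical in-game UI sound class.
# A mixer layers a reduced-volume Wumba container drop foley with a mod bench select sound, both starting immediately.
# A second bench select sound joins the mix through a short fixed delay, staggering the layers.
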SoundCue /WeaponModStation/Sounds/WeaponModStation_ModRemoved_Cue
root=Mixer_0
nodes [
 {id=Delay_0, children [WavePlayer_1], properties {delayMax=0.1, delayMin=0.1}}
 {id=Mixer_0, children [WavePlayer_0,WavePlayer_2,Delay_0], properties {inputVolume [0.65,1,1]}}
 {id=WavePlayer_0, properties {soundWaveAssetPtr=/Game/Sounds/Foley_Loot/Containers/Wumba/Wumba_Drop}}
 {id=WavePlayer_1, properties {soundWaveAssetPtr=/WeaponModStation/Sounds/Wavs/WeaponModBench_Select_01}}
 {id=WavePlayer_2, properties {soundWaveAssetPtr=/WeaponModStation/Sounds/Wavs/WeaponModBench_Select_02}}
]
properties {maxDistance=2097152, soundClassObject=&Fort_UI_InGame_NonCritical, volumeMultiplier=1}
