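# Sound cue for opening the weapon mod station, played as a non-critical in-game UI sound.
# A Random node picks between two equally weighted bench-open wave variants.
# Its concurrency override allows one instance at a time, stopping the oldest when a new one starts.
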SoundCue /WeaponModStation/Sounds/WeaponModStation_Open_Cue
root=Random_0
nodes [
 {id=Random_0, children [WavePlayer_0,WavePlayer_1], properties {weights [1,1]}}
 {id=WavePlayer_0, properties {soundWaveAssetPtr=/WeaponModStation/Sounds/Wavs/WeaponModBench_Open_01}}
 {id=WavePlayer_1, properties {soundWaveAssetPtr=/WeaponModStation/Sounds/Wavs/WeaponModBench_Open_02}}
]
properties {
 bOverrideConcurrency=true
 maxDistance=2097152
 soundClassObject=&Fort_UI_InGame_NonCritical
 volumeMultiplier=1
 concurrencyOverrides {maxCount=1, resolutionRule=StopOldest}
}
