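# UI sound cue for selecting at the Weapon Mod Station, played in the non-critical in-game UI sound class.
# A mixer layers frontend button hover-click and hover waves with a quieter bench-specific select wave.
# The mix feeds a root modulator that holds volume fixed and allows only upward pitch variation.
# A dedicated select concurrency asset governs overlapping plays.
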SoundCue /WeaponModStation/Sounds/WeaponModStation_Select_Cue
root=Modulator_0
nodes [
 {id=Mixer_1, children [WavePlayer_11, WavePlayer_10, WavePlayer_9], properties {inputVolume [1,1.25,0.15]}}
 {id=Modulator_0, children [Mixer_1], properties {pitchMin=1, volumeMax=1, volumeMin=1}}
 {id=WavePlayer_10, properties {soundWaveAssetPtr=/Game/Sounds/Fort_UI_Sounds/FrontEnd/New_FrontEnd/UI_FN_Button_Hover}}
 {id=WavePlayer_11, properties {soundWaveAssetPtr=/Game/Sounds/Fort_UI_Sounds/FrontEnd/New_FrontEnd/UI_FN_Button_Hover_Click_02}}
 {id=WavePlayer_9, properties {soundWaveAssetPtr=/WeaponModStation/Sounds/Wavs/WeaponModBench_Select_02}}
]
properties {
 maxDistance=2097152
 soundClassObject=&Fort_UI_InGame_NonCritical
 volumeMultiplier=1
 concurrencyOverrides {maxCount=1, resolutionRule=StopOldest}
 concurrencySet [&/Paprika_WeaponModStation_Select_Conc]
}
